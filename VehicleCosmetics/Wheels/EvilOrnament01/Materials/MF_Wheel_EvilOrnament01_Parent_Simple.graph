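# Material function for the EvilOrnament01 wheel that extends MF_WheelParent_Simple with an animated emissive effect, apparently a CRT screen look.
# Named reroutes share the rim color, sphere mask, tiled UV1 CRT texture, time and refresh constants.
# Rotated, panned UVs form sawtooth refresh bands, modulated by sine flicker and pulse, with one multiply marked UNFINISHED.
# The result overrides emissive and roughness, then blends with the parent's attributes.
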MaterialFunction /VehicleCosmetics/Wheels/EvilOrnament01/Materials/MF_Wheel_EvilOrnament01_Parent_Simple
graph {nodes [
  n1 = MaterialFunctionCall() {materialFunction=&MF_WheelParent_Simple}
  n2 = TextureCoordinate()
  n3 = Constant() {r=-0.1}
  n4 = NamedRerouteUsage()
  n5 = NamedRerouteUsage()
  n6 = Add(
   A=AppendVector(A=NamedRerouteUsage(), B=NamedRerouteUsage())
   B=Panner(Coordinate=Rotator(Coordinate=@n2, Time=@n3), Time=Multiply(A=@n4, B=@n5)) {speedY=0.5}
  )
  n7 = Add(A=NamedRerouteUsage(), B=ComponentMask(None=@n6) {g=true})
  n8 = Subtract(A=Ceil(None=@n7), B=@n7)
  n9 = NamedRerouteUsage()
  n10 = NamedRerouteUsage()
  n11 = NamedRerouteUsage()
  n12 = Constant2Vector() {g=0.2}
  n13 = Add(A=ComponentMask(None=Add(A=@n6, B=@n12) {constB=0.2}) {g=true}, B=NamedRerouteUsage())
  n14 = Subtract(A=Ceil(None=@n13), B=@n13)
  n15 = LinearInterpolate(
   A=LinearInterpolate(A=NamedRerouteUsage(), B=Multiply(A=@n8, B=@n9), Alpha=@n10)
   B=LinearInterpolate(A=NamedRerouteUsage(), B=Multiply(A=@n11, B=@n14), Alpha=@n10)
   Alpha=NamedRerouteUsage()
  )
  n16 = Multiply(A=@n15) {constB=2}
  n17 = NamedRerouteUsage()
  n18 = NamedRerouteUsage()
  n19 = Sine(None=Add(A=Multiply(A=@n17) {constB=0.1}, B=OneMinus(None=@n18))) {period=0.05}
  n20 = Add(A=Multiply(A=Add(A=@n19) {constB=0.5}) {constB=0.1}) {constB=0.7}
  n21 = Multiply(A=NamedRerouteUsage(), B=Multiply(A=Saturate(None=@n20)) {constB=1.2})
  n22 = Add(A=@n15, B=Saturate(None=NamedRerouteUsage()))
  n23 = NamedRerouteUsage()
  n24 = LinearInterpolate(
   A=NamedRerouteUsage()
   B=Add(A=Multiply(A=@n16, B=@n21) {desc=UNFINISHED}, B=Multiply(A=@n22, B=@n21))
   Alpha=Saturate(None=ConstantBiasScale(None=@n23) {bias=0, scale=100})
  )
  n25 = Multiply(A=NamedRerouteUsage()) {constB=20}
  n26 = Add(A=Multiply(A=Sine(None=@n25)) {constB=0.5}) {constB=0.5}
  n27 = Multiply(A=Saturate(None=ConstantBiasScale(None=@n26) {bias=5, scale=0.09})) {constB=2.5}
  n28 = Multiply(A=Add(A=Multiply(A=@n24, B=@n27)) {constB=0.003}) {constB=1.5}
  n29 = Constant() {r=1}
  FunctionOutput(None=BlendMaterialAttributes(
    A=SetMaterialAttributes!(MaterialAttributes=@n1.MaterialAttributes, EmissiveColor=@n28, Roughness=@n29) {attributeSetTypes [#1, #2]}
    B=@n1.MaterialAttributes
    Alpha=@n1.MaterialAttributes
   ))
  NamedRerouteDeclaration(None=TextureSample() {texture=/VehicleCosmetics/Wheels/EvilOrnament01/Textures/T_Rim_EvilOrnament01_BC}.RGB) {name=Texture_RGB_Out, nodeColor=<rgba 0 0.5781 1 1>}
  n30 = TextureSample() {samplerType=SAMPLERTYPE_LinearColor, texture=/VehicleCosmetics/Wheels/EvilOrnament01/Textures/T_Rim_EvilOrnament01_SphereMask}
  NamedRerouteDeclaration(None=@n30.R) {name=Mask_R_Out, nodeColor=<rgba 0 0.0859 1 1>}
  NamedRerouteDeclaration(None=@n30.G) {name=Mask_G_Out, nodeColor=<rgba 0.4062 1 0 1>}
  NamedRerouteDeclaration(None=Time()) {name=Time_Out, nodeColor=<rgba 1 0 0.8203 1>}
  n31 = TextureSample(UVs=Multiply(A=TextureCoordinate() {coordinateIndex=1, uTiling=12, vTiling=12}) {constB=5}) {texture=/VehicleCosmetics/Wheels/_SharedTextures/FX/T_RGB_CRT}
  NamedRerouteDeclaration(None=@n31.R) {name=CRT_R_Out, nodeColor=<rgba 0.7813 0 1 1>}
  NamedRerouteDeclaration(None=@n31.G) {name=CRT_G_Out, nodeColor=<rgba 0 1 0 1>}
  NamedRerouteDeclaration(None=@n31.B) {name=CRT_B_Out, nodeColor=<rgba 0 0 1 1>}
  NamedRerouteDeclaration(None=@n31.A) {name=CRT_A_Out, nodeColor=<rgba 1 1 1 1>}
  NamedRerouteDeclaration(None=@n31.RGB) {name=CRT_RGB_Out, nodeColor=<rgba 1 0.8931 0 1>}
  NamedRerouteDeclaration(None=Constant() {r=1}) {name=RefreshIntensity_Out, nodeColor=<rgba 1 0.5859 0 1>}
  NamedRerouteDeclaration(None=Constant() {r=1}) {name=RefreshSpeed_Out, nodeColor=<rgba 0.4943 1 0.7204 1>}
 ]}
